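# Looping sound cue for an electric elemental area-of-effect layer, routed to the player elemental damage sound class at reduced volume.
# A quality-level node selects one of four pitch-modulated branches, each looping a standard or "opt" variant of the player electric wave.
# Inline attenuation uses logarithmic falloff with occlusion filtering and listener focus.
# Concurrency caps playback at two instances, stopping the farthest first.
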SoundCue /Game/Sounds/Fort_GamePlay_Sounds/Elemental_Layers/Elemental_ElectricLoop_AOE_Cue
root=QualityLevel_0
nodes [
 {id=Modulator_0, children [WavePlayer_11], properties {pitchMax=1.2}}
 {id=Modulator_2, children [WavePlayer_1], properties {pitchMax=1.2}}
 {id=Modulator_3, children [WavePlayer_0], properties {pitchMax=1.2}}
 {id=Modulator_4, children [WavePlayer_3], properties {pitchMax=1.2}}
 {id=QualityLevel_0, children [Modulator_0,Modulator_2,Modulator_3,Modulator_4]}
 {
  id=WavePlayer_0
  properties {bLooping=true, soundWaveAssetPtr=&/player_on_electric_01_ST_opt}
 }
 {id=WavePlayer_1, properties {bLooping=true, soundWaveAssetPtr=&/player_on_electric_01_ST}}
 {id=WavePlayer_11, properties {bLooping=true, soundWaveAssetPtr=&player_on_electric_01_ST}}
 {
  id=WavePlayer_3
  properties {bLooping=true, soundWaveAssetPtr=&player_on_electric_01_ST_opt}
 }
]
properties {
 attenuationSettings=/Game/Sounds/Attenuation/Fort_Monster_Elemental_Small_Att
 bOverrideAttenuation=true
 bOverrideConcurrency=true
 maxDistance=2000
 soundClassObject=/Game/Sounds/FortSoundClassesAndModes/Fort_ElementalDamage_Player_PSM
 volumeMultiplier=0.5
 attenuationOverrides {
  bEnableListenerFocus=true
  bEnableOcclusion=true
  distanceAlgorithm=Logarithmic
  falloffDistance=1500
  nonFocusVolumeAttenuation=0.3
  occlusionLowPassFilterFrequency=500
  occlusionTraceChannel=ECC_GameTraceChannel11
  occlusionVolumeAttenuation=0.8
  stereoSpread=2000
  attenuationShapeExtents {x=500}
 }
 concurrencyOverrides {maxCount=2, resolutionRule=StopFarthestThenPreventNew}
}
