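# Sound cue for the looping elemental fire layer heard when a player is on fire, routed to the player elemental damage sound class.
# A quality-level node selects one of four pitch-modulated looping wave players: the first two slots use the standard on-fire wave, the last two its "opt" variant.
# Small monster elemental attenuation spatializes it, and a concurrency override caps simultaneous instances, stopping the farthest first.
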SoundCue /Game/Sounds/Fort_GamePlay_Sounds/Elemental_Layers/Elemental_FireLoop_Player_Cue
root=QualityLevel_0
nodes [
 {id=Modulator_0, children [WavePlayer_10], properties {pitchMax=1.2}}
 {id=Modulator_1, children [WavePlayer_0], properties {pitchMax=1.2}}
 {id=Modulator_2, children [WavePlayer_1], properties {pitchMax=1.2}}
 {id=Modulator_3, children [WavePlayer_2], properties {pitchMax=1.2}}
 {id=QualityLevel_0, children [Modulator_0,Modulator_3,Modulator_1,Modulator_2]}
 {id=WavePlayer_0, properties {bLooping=true, soundWaveAssetPtr=&/player_on_fire_01_ST_opt}}
 {id=WavePlayer_1, properties {bLooping=true, soundWaveAssetPtr=&player_on_fire_01_ST_opt}}
 {id=WavePlayer_10, properties {bLooping=true, soundWaveAssetPtr=&/player_on_fire_01_ST}}
 {id=WavePlayer_2, properties {bLooping=true, soundWaveAssetPtr=&player_on_fire_01_ST}}
]
properties {
 attenuationSettings=/Game/Sounds/Attenuation/Fort_Monster_Elemental_Small_Att
 bOverrideConcurrency=true
 maxDistance=1100
 soundClassObject=/Game/Sounds/FortSoundClassesAndModes/Fort_ElementalDamage_Player_PSM
 volumeMultiplier=0.35
 concurrencyOverrides {maxCount=2, resolutionRule=StopFarthestThenPreventNew}
}
