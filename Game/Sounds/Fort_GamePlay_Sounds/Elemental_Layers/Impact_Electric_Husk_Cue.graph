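# Sound cue for an electric elemental impact layer on husks, routed through the melee weapon sound class with small melee attenuation.
# A Random node picks one of four equally weighted husk electrocute zap waves.
# A Modulator at the root applies its default pitch and volume variation to the chosen wave, so repeated hits differ.
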SoundCue /Game/Sounds/Fort_GamePlay_Sounds/Elemental_Layers/Impact_Electric_Husk_Cue
root=Modulator_2
nodes [
 {id=Modulator_2, children [Random_1]}
 {id=Random_1, children [WavePlayer_5,WavePlayer_6,WavePlayer_7,WavePlayer_8], properties {weights [1,1,1,1]}}
 {id=WavePlayer_5, properties {soundWaveAssetPtr=/Game/Sounds/Fort_Gadgets/Electrical_Zaps/husk_electrocute_08}}
 {id=WavePlayer_6, properties {soundWaveAssetPtr=/Game/Sounds/Fort_Gadgets/Electrical_Zaps/husk_electrocute_03}}
 {id=WavePlayer_7, properties {soundWaveAssetPtr=/Game/Sounds/Fort_Gadgets/Electrical_Zaps/husk_electrocute_05}}
 {id=WavePlayer_8, properties {soundWaveAssetPtr=/Game/Sounds/Fort_Gadgets/Electrical_Zaps/husk_electrocute_06}}
]
properties {
 attenuationSettings=&Fort_Weapon_Melee_Sml_Att
 maxDistance=2800
 soundClassObject=&Fort_WeaponMelee_PSM
 volumeMultiplier=1
}
