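# Sound cue for the fire elemental impact layer on Husk enemies, playing a short fire impact sound.
# A Random node picks among six equally weighted enemy fire short impact waves and feeds a Modulator at the root for per-play variation.
# It uses the small melee weapon attenuation and the melee weapon sound class.
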SoundCue /Game/Sounds/Fort_GamePlay_Sounds/Elemental_Layers/Impact_Fire_Husk_Cue
root=Modulator_2
nodes [
 {id=Modulator_2, children [Random_0]}
 {id=Random_0, children [WavePlayer_1,WavePlayer_2,WavePlayer_3,WavePlayer_4,WavePlayer_6,WavePlayer_7], properties {weights [1,1,1,1,1,1]}}
 {id=WavePlayer_1, properties {soundWaveAssetPtr=&/enemy_fire_short_impact_01}}
 {id=WavePlayer_2, properties {soundWaveAssetPtr=&/enemy_fire_short_impact_02}}
 {id=WavePlayer_3, properties {soundWaveAssetPtr=&/enemy_fire_short_impact_03}}
 {id=WavePlayer_4, properties {soundWaveAssetPtr=&/enemy_fire_short_impact_04}}
 {id=WavePlayer_6, properties {soundWaveAssetPtr=&/enemy_fire_short_impact_06}}
 {id=WavePlayer_7, properties {soundWaveAssetPtr=&/enemy_fire_short_impact_07}}
]
properties {
 attenuationSettings=&Fort_Weapon_Melee_Sml_Att
 maxDistance=2800
 soundClassObject=&Fort_WeaponMelee_PSM
 volumeMultiplier=1
}
