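# Sound cue for the ice elemental impact layer for husks, using small enemy ice impact sounds.
# A Random node picks one of five wave variations with equal weight and feeds a Modulator at the root for pitch and volume variation.
# It uses the small melee weapon attenuation and the melee weapon sound class.
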SoundCue /Game/Sounds/Fort_GamePlay_Sounds/Elemental_Layers/Impact_Ice_Husk_Cue
root=Modulator_2
nodes [
 {id=Modulator_2, children [Random_0]}
 {id=Random_0, children [WavePlayer_4,WavePlayer_3,WavePlayer_2,WavePlayer_1,WavePlayer_0], properties {weights [1,1,1,1,1]}}
 {id=WavePlayer_0, properties {soundWaveAssetPtr=&/enemy_ice_small_impact_ST_09}}
 {id=WavePlayer_1, properties {soundWaveAssetPtr=&/enemy_ice_small_impact_ST_06}}
 {id=WavePlayer_2, properties {soundWaveAssetPtr=&/enemy_ice_small_impact_ST_05}}
 {id=WavePlayer_3, properties {soundWaveAssetPtr=&/enemy_ice_small_impact_ST_04}}
 {id=WavePlayer_4, properties {soundWaveAssetPtr=&/enemy_ice_small_impact_ST_03}}
]
properties {
 attenuationSettings=&Fort_Weapon_Melee_Sml_Att
 maxDistance=2800
 soundClassObject=&Fort_WeaponMelee_PSM
}
